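# Niagara base system for fountain water splashes, placed on configurable concentric or offset rings.
# Enabled TopSplash (GPU, distance-culled) and Splash_SubUV sprites partition particles by ring and beam count, using user arrays for per-ring radius and offset.
# User Alpha scales opacity, curves drive dynamic material parameters, and TopSplash tints by time of day through an NPC parameter.
# Five emitters are disabled: drop splashes, edge flow, lingering spray and a ribbon water stream.
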NiagaraSystem /Game/Effects/Fort_Effects/Effects/Environment/Fountain/NS_Fountain_Base
dependencies {
 usedDataInterfaces [&NiagaraDataInterfaceCurve, &NiagaraDataInterfaceColorCurve, &NiagaraDataInterfaceVector2DCurve]
 usedDynamicInputs []
 usedModules [&SystemState]
}
emitters {
 DropSplash {
  bEnabled=false
  simTarget=GPUComputeSim
  emitterUpdateScript [
   &EmitterState(
    Enable Distance Culling=true
    Life Cycle Mode=::NewEnumerator1
    Loop Duration=2
    MaxDistance=12500
    Module.bUseMaxDistance=true
    Module.bUseMinDistance=false
    Scalability Mode=::NewEnumerator1
    UseLoopDelay=false
   )
   &SpawnRate(
    SpawnRate=&Multiply_Float(A=100, B=&MakeFloatFromInt(Int=$User.NumberofRings))
    Use Spawn Probability=false
   )
  ]
  particleSpawnScript [
   &InitializeParticle(
    Color Minimum=<rgba 1 1 1 1>
    Color Mode=::NewEnumerator2
    Lifetime Max=0.6
    Lifetime Min=0.5
    Lifetime Mode=::NewEnumerator1
    Mass Mode=::NewEnumerator3
    Sprite Rotation Mode=::NewEnumerator2
    Sprite Size Max=<15 60>
    Sprite Size Min=<15 50>
    Sprite Size Mode=::NewEnumerator2
    UsePositionOffset=false
    Write Material Random=true
   )
   &PartitionParticles(
    Number Of Partitions=$User.NumberofRings
    Particles Per Partition:NiagaraInt32=$User.NumberofBeams
    Partition Mode:ENiagara_PartitionMode=::NewEnumerator1
   )
   &ShapeLocation(
    MaskSpawnGroup=false
    Offset=&SelectVectorFromArray(Direct Array Index=$Particles.PartitionParticles.Partition, Vector Selection Array=$User.RingOffset)
    Offset Coordinate Space=::Local
    Offset Mode=::NewEnumerator0
    Override Seed=false
    Ring / Disc Distribution Mode:ENiagaraSphereDistributionMode=::NewEnumerator1
    Ring Radius=&Multiply_Float(
     A=&SelectFloatFromArray(Direct Array Index=$Particles.PartitionParticles.Partition, Float Selection Array=$User.RingSize)
     B=$User.BaseRadiusScale
    )
    Shape Primitive=::NewEnumerator4
    U Position=$Particles.PartitionParticles.Normalized Index In Partition
    UseOwnerScale=true
   )
   &AddVelocity(Velocity=<0 0 0>)
   &SubUVAnimation(End Frame=3, SubUV Animation Mode:ENiagara_SubUVLookupMode=::NewEnumerator2)
  ]
  particleUpdateScript [
   &ParticleState()
   &ScaleSpriteSize(Uniform Curve Sprite Scale=DataInterface() {propertyValues {Curve {keys [
        {arriveTangent=5.6242, interpMode=RCIM_Cubic, leaveTangent=5.6242, tangentMode=RCTM_User, time=0, value=0}
        {arriveTangent=0.9995, interpMode=RCIM_Cubic, leaveTangent=0.9995, tangentMode=RCTM_User, time=0.5, value=1}
       ]}}})
   &GravityForce(Gravity=&RandomRangeVector(Minimum=<0 0 -300>, Maximum=<1 1 -400>))
   &Drag(Drag=0.3, Use Linear Drag=true, Use Rotational Drag=false)
   &ScaleColor(
    Scale Alpha=&FloatFromCurve(
     FloatCurve=DataInterface() {propertyValues {Curve {keys [
         {arriveTangent=6.213, interpMode=RCIM_Cubic, leaveTangent=6.213, tangentMode=RCTM_User, time=0, value=0}
         {arriveTangent=-1.5422, interpMode=RCIM_Cubic, leaveTangent=-1.5422, tangentMode=RCTM_User, time=0.4, value=1.075}
         {arriveTangent=-1.4513, interpMode=RCIM_Cubic, tangentMode=RCTM_Break, time=1, value=0}
        ]}}}
     Scale Curve=$User.Alpha
    )
    Scale RGB=&Multiply_Vector(
     A=<0.25 0.25 0.25>
     B=&MakeVectorFromLinearColorRGB(Color=&ColorFromCurve(
       ColorCurve=DataInterface() {propertyValues {
         AlphaCurve {keys [<at 0 1>,<at 1 1>]}
         BlueCurve {keys [{arriveTangent=0.256, arriveTangentWeight=0.0864, interpMode=RCIM_Cubic, leaveTangentWeight=0.0864, time=0.001, value=0.2157}, <at 0.4988 1>]}
         GreenCurve {keys [
           {arriveTangent=0.2383, arriveTangentWeight=0.0668, interpMode=RCIM_Cubic, leaveTangentWeight=0.0668, time=0.001, value=0.1765}
           <at 0.4988 1>
          ]}
         RedCurve {keys [
           {arriveTangent=0.2622, arriveTangentWeight=0.0786, interpMode=RCIM_Cubic, leaveTangentWeight=0.0786, time=0.001, value=0.1098}
           <at 0.4988 1>
          ]}
        }}
       CurveIndex=Linked() {linkedVariable {name=NPC.FortniteNPC.TimeOfDay_0-1, type=NiagaraFloat}}
      ))
    )
    ScaleA=true
    ScaleRGB=true
   )
   &SolveForcesAndVelocity(Clamp Velocity=false, Limit Acceleration=false)
   &DynamicMaterialParameters(
    Index 0 Param 1=&FloatFromCurve(FloatCurve=DataInterface() {propertyValues {Curve {keys [{interpMode=RCIM_Cubic, leaveTangent=1.7548, tangentMode=RCTM_Break, time=0, value=0}, {tangentMode=RCTM_User, time=1, value=1}]}}})
    Module.Param0WriteEnabled=true
    Module.Param1WriteEnabled=true
    Module.Param2WriteEnabled=true
    Module.Param3WriteEnabled=true
   )
   &CameraOffset(Camera Offset Amount=10)
  ]
  rendererClasses [&NiagaraSpriteRendererProperties]
  renderers [{rendererClass=&NiagaraSpriteRendererProperties}]
 }
 DropSplash001 {
  bEnabled=false
  simTarget=GPUComputeSim
  emitterUpdateScript [
   &EmitterState(
    Enable Distance Culling=true
    Life Cycle Mode=::NewEnumerator1
    Loop Duration=2
    MaxDistance=12500
    Module.bUseMaxDistance=true
    Module.bUseMinDistance=false
    Scalability Mode=::NewEnumerator1
    UseLoopDelay=false
   )
   &SpawnRate(SpawnRate=&Multiply_Float(A=15, B=$User.TopSpraySpawnRate), Use Spawn Probability=false)
  ]
  particleSpawnScript [
   &InitializeParticle(
    Color Minimum=<rgba 1 1 1 1>
    Color Mode=::NewEnumerator2
    Lifetime Max=0.8
    Lifetime Min=0.6
    Lifetime Mode=::NewEnumerator1
    Mass Mode=::NewEnumerator3
    Sprite Rotation Mode=::NewEnumerator2
    Sprite Size Max=<15 100>
    Sprite Size Min=<15 50>
    Sprite Size Mode=::NewEnumerator2
    UsePositionOffset=false
    Write Material Random=true
   )
   &AddVelocity(Velocity=<0 0 150>)
   &SubUVAnimation(End Frame=3, SubUV Animation Mode:ENiagara_SubUVLookupMode=::NewEnumerator2)
   &SystemLocation(MaskSpawnGroup=false, Offset=&MakeVector(X=0, Y=0, Z=$User.TopSprayPosition))
  ]
  particleUpdateScript [
   &ParticleState()
   &ScaleSpriteSize(Uniform Curve Sprite Scale=DataInterface() {propertyValues {Curve {keys [
        {arriveTangent=5.6242, interpMode=RCIM_Cubic, leaveTangent=5.6242, tangentMode=RCTM_User, time=0, value=0}
        {arriveTangent=0.9995, interpMode=RCIM_Cubic, leaveTangent=0.9995, tangentMode=RCTM_User, time=0.5, value=1}
       ]}}})
   &GravityForce(Gravity=&RandomRangeVector(Minimum=<0 0 -100>, Maximum=<1 1 -100>))
   &Drag(Drag=0.3, Use Linear Drag=true, Use Rotational Drag=false)
   &ScaleColor(
    Scale Alpha=&FloatFromCurve(
     FloatCurve=DataInterface() {propertyValues {Curve {keys [
         {arriveTangent=6.213, interpMode=RCIM_Cubic, leaveTangent=6.213, tangentMode=RCTM_User, time=0, value=0}
         {arriveTangent=-1.5422, interpMode=RCIM_Cubic, leaveTangent=-1.5422, tangentMode=RCTM_User, time=0.4, value=1.075}
         {arriveTangent=-1.4513, interpMode=RCIM_Cubic, tangentMode=RCTM_Break, time=1, value=0}
        ]}}}
     Scale Curve=&Multiply_Float(A=2, B=$User.Alpha)
    )
    Scale RGB=&Multiply_Vector(
     A=<0.25 0.25 0.25>
     B=&MakeVectorFromLinearColorRGB(Color=&ColorFromCurve(
       ColorCurve=DataInterface() {propertyValues {
         AlphaCurve {keys [<at 0 1>,<at 1 1>]}
         BlueCurve {keys [{arriveTangent=0.256, arriveTangentWeight=0.0864, interpMode=RCIM_Cubic, leaveTangentWeight=0.0864, time=0.001, value=0.2157}, <at 0.4988 1>]}
         GreenCurve {keys [
           {arriveTangent=0.2383, arriveTangentWeight=0.0668, interpMode=RCIM_Cubic, leaveTangentWeight=0.0668, time=0.001, value=0.1765}
           <at 0.4988 1>
          ]}
         RedCurve {keys [
           {arriveTangent=0.2622, arriveTangentWeight=0.0786, interpMode=RCIM_Cubic, leaveTangentWeight=0.0786, time=0.001, value=0.1098}
           <at 0.4988 1>
          ]}
        }}
       CurveIndex=Linked() {linkedVariable {name=NPC.FortniteNPC.TimeOfDay_0-1, type=NiagaraFloat}}
      ))
    )
    ScaleA=true
    ScaleRGB=true
   )
   &SolveForcesAndVelocity(Clamp Velocity=false, Limit Acceleration=false)
   &DynamicMaterialParameters(
    Index 0 Param 1=&FloatFromCurve(FloatCurve=DataInterface() {propertyValues {Curve {keys [{interpMode=RCIM_Cubic, leaveTangent=1.7548, tangentMode=RCTM_Break, time=0, value=0}, {tangentMode=RCTM_User, time=1, value=1}]}}})
    Module.Param0WriteEnabled=true
    Module.Param1WriteEnabled=true
    Module.Param2WriteEnabled=true
    Module.Param3WriteEnabled=true
   )
  ]
  rendererClasses [&NiagaraSpriteRendererProperties]
  renderers [{rendererClass=&NiagaraSpriteRendererProperties}]
 }
 EdgeFlow {
  bEnabled=false
  simTarget=GPUComputeSim
  emitterUpdateScript [
   &EmitterState(Life Cycle Mode=::NewEnumerator1, Loop Duration=2, UseLoopDelay=false)
   &SpawnRate(
    SpawnRate=&Multiply_Float(A=10, B=&MakeFloatFromInt(Int=$User.NumberofRings))
    Use Spawn Probability=false
   )
  ]
  particleSpawnScript [
   &InitializeParticle(
    Position Mode=::NewEnumerator1
    Position Offset Coordinate Space=::Local
    Sprite Rotation Angle=-90
    Sprite Rotation Mode=::NewEnumerator0
    Sprite Size Max=<150 25>
    Sprite Size Min=<100 15>
    Sprite Size Mode=::NewEnumerator2
    UsePositionOffset=true
    Write Material Random=true
   )
   &PartitionParticles(
    Number Of Partitions=$User.NumberofRings
    Particles Per Partition:NiagaraInt32=$User.NumberofBeams
    Partition Mode:ENiagara_PartitionMode=::NewEnumerator1
   )
   &ShapeLocation(
    MaskSpawnGroup=false
    Offset=&SelectVectorFromArray(Direct Array Index=$Particles.PartitionParticles.Partition, Vector Selection Array=$User.RingOffset)
    Offset Coordinate Space=::Local
    Offset Mode=::NewEnumerator0
    Override Seed=false
    Ring / Disc Distribution Mode:ENiagaraSphereDistributionMode=::NewEnumerator1
    Ring Radius=&Multiply_Float(
     A=&SelectFloatFromArray(Direct Array Index=$Particles.PartitionParticles.Partition, Float Selection Array=$User.RingSize)
     B=$User.BaseRadiusScale
    )
    Shape Primitive=::NewEnumerator4
    U Position=$Particles.PartitionParticles.Normalized Index In Partition
    UseOwnerScale=true
   )
  ]
  particleUpdateScript [
   &ParticleState()
   &SpriteFacingAndAlignment(Sprite Alignment=<0 0 1>, WriteAlignment=true, WriteFacing=false)
   &ScaleColor(
    Scale Alpha=&FloatFromCurve(
     FloatCurve=DataInterface() {propertyValues {Curve {keys [
         {arriveTangent=6.213, interpMode=RCIM_Cubic, leaveTangent=6.213, tangentMode=RCTM_User, time=0, value=0}
         {arriveTangent=-1.5422, interpMode=RCIM_Cubic, leaveTangent=-1.5422, tangentMode=RCTM_User, time=0.4, value=1.075}
         {arriveTangent=-1.4513, interpMode=RCIM_Cubic, tangentMode=RCTM_Break, time=1, value=0}
        ]}}}
     Scale Curve=$User.Alpha
    )
    Scale RGB=&Multiply_Vector(
     A=<0.25 0.25 0.25>
     B=&MakeVectorFromLinearColorRGB(Color=&ColorFromCurve(
       ColorCurve=DataInterface() {propertyValues {
         AlphaCurve {keys [<at 0 1>,<at 1 1>]}
         BlueCurve {keys [{arriveTangent=0.256, arriveTangentWeight=0.0864, interpMode=RCIM_Cubic, leaveTangentWeight=0.0864, time=0.001, value=0.2157}, <at 0.4988 1>]}
         GreenCurve {keys [
           {arriveTangent=0.2383, arriveTangentWeight=0.0668, interpMode=RCIM_Cubic, leaveTangentWeight=0.0668, time=0.001, value=0.1765}
           <at 0.4988 1>
          ]}
         RedCurve {keys [
           {arriveTangent=0.2622, arriveTangentWeight=0.0786, interpMode=RCIM_Cubic, leaveTangentWeight=0.0786, time=0.001, value=0.1098}
           <at 0.4988 1>
          ]}
        }}
       CurveIndex=Linked() {linkedVariable {name=NPC.FortniteNPC.TimeOfDay_0-1, type=NiagaraFloat}}
      ))
    )
    ScaleA=true
    ScaleRGB=true
   )
  ]
  rendererClasses [&NiagaraSpriteRendererProperties]
  renderers [{rendererClass=&NiagaraSpriteRendererProperties}]
 }
 SplashLinger001 {
  bEnabled=false
  emitterUpdateScript [
   &EmitterState()
   &SpawnRate(SpawnRate=&Multiply_Float(A=50, B=$User.TopSpraySpawnRate), Use Spawn Probability=false)
  ]
  particleSpawnScript [
   &InitializeParticle(
    Lifetime Max=0.7
    Lifetime Min=0.5
    Lifetime Mode=::NewEnumerator1
    Sprite Rotation Mode=::NewEnumerator2
    Sprite Size Mode=::NewEnumerator1
    Uniform Sprite Size Max=65
    Uniform Sprite Size Min=35
    UsePositionOffset=true
    Write Material Random=true
   )
   &SubUVAnimation(End Frame=3, SubUV Animation Mode:ENiagara_SubUVLookupMode=::NewEnumerator2)
   &AddVelocity(
    Cone Angle=5
    Cone Axis=<0 0 1>
    Module.UseConeFalloff=false
    Module.UseVelDistribution=false
    Override Seed=false
    Velocity Mode=::NewEnumerator2
    Velocity Speed=&RandomRangeFloat(Minimum=100, Maximum=300)
   )
   &CameraOffset(Camera Offset Amount=20)
   &SystemLocation(
    MaskSpawnGroup=false
    Offset=&MakeVector(X=0, Y=0, Z=&Add_Float(A=135, B=$User.TopSprayPosition))
   )
  ]
  particleUpdateScript [
   &Drag(Use Linear Drag=true, Use Rotational Drag=false)
   &GravityForce(Gravity=<0 0 -800>)
   &SolveForcesAndVelocity(Clamp Velocity=false, Limit Acceleration=false)
   &ParticleState()
   &ScaleSpriteSize(Uniform Curve Sprite Scale=DataInterface() {propertyValues {Curve {keys [{interpMode=RCIM_Cubic, tangentMode=RCTM_User, time=0, value=0.1}, {arriveTangent=1.1797, tangentMode=RCTM_User, time=1, value=1}]}}})
   &ScaleColor(
    Scale Alpha=&Multiply_Float(A=$User.Alpha, B=5)
    Scale RGB=&Multiply_VectorByFloat(
     Vector=<1 1 1>
     Float=&FloatFromCurve(FloatCurve=DataInterface() {propertyValues {Curve {keys [<at 0 1>,<at 1 0>]}}})
    )
    ScaleA=true
    ScaleRGB=true
   )
   &DynamicMaterialParameters(
    Index 0 Param 1=&FloatFromCurve(FloatCurve=DataInterface() {propertyValues {Curve {keys [
         {arriveTangent=-0.7772, interpMode=RCIM_Cubic, leaveTangent=-0.7772, tangentMode=RCTM_User, time=0, value=1}
         <at 0.7 1>
         {arriveTangent=-2.6733, interpMode=RCIM_Cubic, leaveTangent=-2.6733, tangentMode=RCTM_User, time=1, value=0}
        ]}}})
    Module.Param0WriteEnabled=true
    Module.Param1WriteEnabled=true
    Module.Param2WriteEnabled=true
    Module.Param3WriteEnabled=true
   ) {module=DynamicMaterialParameters001}
  ]
  rendererClasses [&NiagaraSpriteRendererProperties]
  renderers [{rendererClass=&NiagaraSpriteRendererProperties}]
 }
 Splash_SubUV {
  emitterUpdateScript [
   &EmitterState()
   &SpawnRate(
    SpawnRate=&Multiply_Float(A=150, B=&MakeFloatFromInt(Int=$User.NumberofRings))
    Use Spawn Probability=false
   )
  ]
  particleSpawnScript [
   &InitializeParticle(
    Lifetime Max=0.7
    Lifetime Min=0.3
    Lifetime Mode=::NewEnumerator1
    Sprite Size Max=<50 35>
    Sprite Size Min=<25 15>
    Sprite Size Mode=::NewEnumerator2
    UsePositionOffset=true
    Write Material Random=true
   ) {module=InitializeParticle001}
   &AddVelocityInCone(
    Cone Angle=50
    Cone Axis=<0 0 1>
    Module.Use Velocity Falloff On Cone Axis=false
    Velocity Distribution Along Cone Axis=0.25
    Velocity Strength=&UniformRangedFloat(Minimum=35, Maximum=100)
   )
   &SystemLocation(MaskSpawnGroup=false, Offset=<0 0 -125>)
   &PartitionParticles(
    Number Of Partitions=$User.NumberofRings
    Particles Per Partition:NiagaraInt32=$User.NumberofBeams
    Partition Mode:ENiagara_PartitionMode=::NewEnumerator1
   )
   &ShapeLocation(
    MaskSpawnGroup=false
    Offset=&SelectVectorFromArray(Direct Array Index=$Particles.PartitionParticles.Partition, Vector Selection Array=$User.SplashOffset)
    Offset Coordinate Space=::Local
    Offset Mode=::NewEnumerator0
    Override Seed=false
    Ring / Disc Distribution Mode:ENiagaraSphereDistributionMode=::NewEnumerator1
    Ring Radius=&Multiply_Float(
     A=&SelectFloatFromArray(Direct Array Index=$Particles.PartitionParticles.Partition, Float Selection Array=$User.RingSize)
     B=$User.BaseRadiusScale
    )
    Shape Primitive=::NewEnumerator4
    U Position=$Particles.PartitionParticles.Normalized Index In Partition
    UseOwnerScale=true
   )
   &SubUVAnimation(End Frame=3, SubUV Animation Mode:ENiagara_SubUVLookupMode=::NewEnumerator2)
  ]
  particleUpdateScript [
   &ParticleState()
   &ScaleSpriteSize(Scale Factor=&Vector2DFromCurve(
     Vector2Curve=DataInterface() {propertyValues {
       XCurve {keys [
         {arriveTangent=0.4604, interpMode=RCIM_Cubic, leaveTangent=0.4604, tangentMode=RCTM_User, time=0, value=0.05}
         {arriveTangent=1.2987, interpMode=RCIM_Cubic, leaveTangent=1.2987, tangentMode=RCTM_User, time=1, value=1}
        ]}
       YCurve {keys [
         {arriveTangent=2.0407, interpMode=RCIM_Cubic, leaveTangent=2.0407, tangentMode=RCTM_User, time=0, value=0.05}
         {arriveTangentWeight=0.3333, interpMode=RCIM_Cubic, leaveTangentWeight=0.0833, tangentMode=RCTM_User, time=0.75, value=1}
         {arriveTangent=-2.358, interpMode=RCIM_Cubic, leaveTangent=-2.358, tangentMode=RCTM_User, time=1, value=0.7}
        ]}
      }}
     ScaleCurve=<1 1>
    ))
   &DynamicMaterialParameters(
    Index 0 Param 1=&FloatFromCurve(
     FloatCurve=DataInterface() {propertyValues {Curve {keys [
         {arriveTangent=-0.5095, interpMode=RCIM_Cubic, leaveTangent=-0.5095, tangentMode=RCTM_User, time=0, value=1}
         {arriveTangent=-1.3585, interpMode=RCIM_Cubic, leaveTangent=-1.3585, tangentMode=RCTM_User, time=1, value=0}
        ]}}}
     Scale Curve=1.5
    )
    Module.Param0WriteEnabled=true
    Module.Param1WriteEnabled=false
    Module.Param2WriteEnabled=false
    Module.Param3WriteEnabled=false
   )
   &Color(Color=&MakeLinearColorFromVectorAndFloat(
     "Vector (RGB)"=<1 1 1>
     "Float (Alpha)"=&FloatFromCurve(
      FloatCurve=DataInterface() {propertyValues {Curve {keys [<at 0.4 1.0113 RCIM_Cubic>,<at 1.0011 0.002>]}}}
      Scale Curve=&Multiply_Float(A=$User.Alpha, B=5)
     )
    ))
   &Drag(Drag=2, Use Linear Drag=true, Use Rotational Drag=false)
   &WindForce(Air Resistance=0.1, Wind Speed=<0 0 -1>, Wind Speed Scale=100)
   &SpriteRotationRate()
   &SolveForcesAndVelocity(Clamp Velocity=false, Limit Acceleration=false)
  ]
  rendererClasses [&NiagaraSpriteRendererProperties]
  renderers [{rendererClass=&NiagaraSpriteRendererProperties}]
 }
 TopSplash {
  simTarget=GPUComputeSim
  emitterUpdateScript [
   &EmitterState(
    Enable Distance Culling=true
    Life Cycle Mode=::NewEnumerator1
    Loop Duration=2
    MaxDistance=10000
    Module.bUseMaxDistance=true
    Module.bUseMinDistance=false
    Scalability Mode=::NewEnumerator1
    UseLoopDelay=false
   )
   &SpawnRate(
    SpawnRate=&Multiply_Float(A=50, B=&MakeFloatFromInt(Int=$User.NumberofRings))
    Use Spawn Probability=false
   )
  ]
  particleSpawnScript [
   &InitializeParticle(
    Color Maximum=<rgba 1 1 1 1.5>
    Color Minimum=<rgba 1 1 1 1>
    Color Mode=::NewEnumerator2
    Lifetime Max=0.7
    Lifetime Min=0.6
    Lifetime Mode=::NewEnumerator1
    Mass Mode=::NewEnumerator3
    Sprite Rotation Mode=::NewEnumerator2
    Sprite Size Mode=::NewEnumerator1
    Uniform Sprite Size Max=40
    Uniform Sprite Size Min=25
    UsePositionOffset=false
    Write Material Random=true
   )
   &SystemLocation(MaskSpawnGroup=false, Offset=<0 0 -25>)
   &PartitionParticles(
    Number Of Partitions=$User.NumberofRings
    Particles Per Partition:NiagaraInt32=$User.NumberofBeams
    Partition Mode:ENiagara_PartitionMode=::NewEnumerator1
   )
   &ShapeLocation(
    MaskSpawnGroup=false
    Offset=&SelectVectorFromArray(Direct Array Index=$Particles.PartitionParticles.Partition, Vector Selection Array=$User.RingOffset)
    Offset Coordinate Space=::Local
    Offset Mode=::NewEnumerator0
    Override Seed=false
    Ring / Disc Distribution Mode:ENiagaraSphereDistributionMode=::NewEnumerator1
    Ring Radius=&Multiply_Float(
     A=&SelectFloatFromArray(Direct Array Index=$Particles.PartitionParticles.Partition, Float Selection Array=$User.RingSize)
     B=$User.BaseRadiusScale
    )
    Shape Primitive=::NewEnumerator4
    U Position=$Particles.PartitionParticles.Normalized Index In Partition
    UseOwnerScale=true
   )
   &AddVelocity(Velocity=<0 0 -25>)
   &SubUVAnimation(End Frame=3)
  ]
  particleUpdateScript [
   &ParticleState()
   &ScaleSpriteSize(Uniform Curve Sprite Scale=DataInterface() {propertyValues {Curve {keys [
        {interpMode=RCIM_Cubic, leaveTangent=5.6407, tangentMode=RCTM_Break, time=0, value=0.4}
        {tangentMode=RCTM_User, time=0.3, value=1}
        <at 1 1.5>
       ]}}})
   &GravityForce(Gravity=<0 0 -200>)
   &ScaleColor(
    Scale Alpha=&FloatFromCurve(
     FloatCurve=DataInterface() {propertyValues {Curve {keys [
         {arriveTangent=6.213, interpMode=RCIM_Cubic, leaveTangent=6.213, tangentMode=RCTM_User, time=0, value=0}
         {arriveTangent=-1.5422, interpMode=RCIM_Cubic, leaveTangent=-1.5422, tangentMode=RCTM_User, time=0.4, value=1.075}
         {arriveTangent=-1.4513, interpMode=RCIM_Cubic, tangentMode=RCTM_Break, time=1, value=0}
        ]}}}
     Scale Curve=$User.Alpha
    )
    Scale RGB=&Multiply_Vector(
     A=<0.25 0.25 0.25>
     B=&MakeVectorFromLinearColorRGB(Color=&ColorFromCurve(
       ColorCurve=DataInterface() {propertyValues {
         AlphaCurve {keys [<at 0 1>,<at 1 1>]}
         BlueCurve {keys [{arriveTangent=0.256, arriveTangentWeight=0.0864, interpMode=RCIM_Cubic, leaveTangentWeight=0.0864, time=0.001, value=0.2157}, <at 0.4988 1>]}
         GreenCurve {keys [
           {arriveTangent=0.2383, arriveTangentWeight=0.0668, interpMode=RCIM_Cubic, leaveTangentWeight=0.0668, time=0.001, value=0.1765}
           <at 0.4988 1>
          ]}
         RedCurve {keys [
           {arriveTangent=0.2622, arriveTangentWeight=0.0786, interpMode=RCIM_Cubic, leaveTangentWeight=0.0786, time=0.001, value=0.1098}
           <at 0.4988 1>
          ]}
        }}
       CurveIndex=Linked() {linkedVariable {name=NPC.FortniteNPC.TimeOfDay_0-1, type=NiagaraFloat}}
      ))
    )
    ScaleA=true
    ScaleRGB=true
   )
   &SolveForcesAndVelocity(Clamp Velocity=false, Limit Acceleration=false)
   &DynamicMaterialParameters(
    Index 0 Param 1=&FloatFromCurve(FloatCurve=DataInterface() {propertyValues {Curve {keys [
         {interpMode=RCIM_Cubic, leaveTangent=-0.7415, tangentMode=RCTM_Break, time=0, value=1}
         {arriveTangent=-1.5521, interpMode=RCIM_Cubic, leaveTangent=-8.0089, tangentMode=RCTM_Break, time=1, value=0}
        ]}}})
    Index 0 Param 2=&FloatFromCurve(
     FloatCurve=DataInterface() {propertyValues {Curve {keys [
         {interpMode=RCIM_Cubic, leaveTangent=-0.4439, tangentMode=RCTM_Break, time=0, value=0.2}
         {arriveTangent=-0.0393, interpMode=RCIM_Cubic, tangentMode=RCTM_Break, time=1, value=1}
        ]}}}
     Scale Curve=0.7
    )
    Index 0 Param 3=&FloatFromCurve(
     FloatCurve=DataInterface() {propertyValues {Curve {keys [{interpMode=RCIM_Cubic, leaveTangent=2.6753, tangentMode=RCTM_Break, time=0.5, value=0}, <at 1 1>]}}}
     Scale Curve=3
    )
    Index 0 Param 4=&FloatFromCurve(FloatCurve=DataInterface() {propertyValues {Curve {keys [{interpMode=RCIM_Cubic, leaveTangent=2.6753, tangentMode=RCTM_Break, time=0, value=0}, <at 1 1>]}}})
    Index 1 Param 1=&FloatFromCurve(FloatCurve=DataInterface() {propertyValues {Curve {keys [{interpMode=RCIM_Cubic, tangentMode=RCTM_User, time=0, value=1}, {interpMode=RCIM_Cubic, tangentMode=RCTM_User, time=0.6, value=0}]}}})
    Module.Param0WriteEnabled=true
    Module.Param0WriteEnabled1=true
    Module.Param1WriteEnabled=true
    Module.Param1WriteEnabled1=true
    Module.Param2WriteEnabled=true
    Module.Param2WriteEnabled1=true
    Module.Param3WriteEnabled=true
    Module.Param3WriteEnabled1=true
    Write Parameter Index 1=true
   )
  ]
  rendererClasses [&NiagaraSpriteRendererProperties]
  renderers [{rendererClass=&NiagaraSpriteRendererProperties}]
 }
 WaterStream003 {
  bEnabled=false
  emitterUpdateScript [
   &EmitterState(
    Life Cycle Mode=::NewEnumerator1
    Loop Duration=&UniformRangedFloat(Minimum=1, Maximum=1)
    Recalculate Duration Each Loop=true
    UseLoopDelay=false
   )
   &SpawnRate(
    SpawnRate=&Multiply_Float(
     A=&Multiply_Float(A=25, B=$User.TopSpraySpawnRate)
     B=&MakeFloatFromInt(Int=$User.NumberofRings)
    )
    Use Spawn Probability=false
   )
  ]
  particleSpawnScript [
   /Niagara/Modules/Spawn/Initialization/InitializeParticle(
    Lifetime=0.75
    Module.Write Color=true
    Module.Write Lifetime=true
    Module.Write Mass=true
    Module.Write Material Random=true
    Module.Write Position=false
    Module.Write Scale=false
    Module.Write SpriteRotation=false
    Module.Write SpriteSize=true
    Position=&SimulationPosition()
   ) {enabled=false}
   &InitializeRibbon(
    Color=&MultiplyLinearColor(
     Linear Color=<rgba 1 1 1 1>
     Scale Factor=&MakeFloat4FromVectorAndFloat("Vector (RGB)"=<1 1 1>, "Float (Alpha)"=&Lerp_Float(A=1, B=1, Alpha=$System.WaterForce))
    )
    Ribbon Width=15
    UsePositionOffset=false
    Write Material Random=true
   )
   &AddVelocity(Rotation Coordinate Space=::Simulation, Velocity=<0 0 300>)
   &SystemLocation(MaskSpawnGroup=false, Offset=&MakeVector(X=0, Y=0, Z=$User.TopSprayPosition))
  ]
  particleUpdateScript [
   &ParticleState()
   &CurlNoiseForce(Noise Frequency=25, Noise Strength=5, Pan Noise=true, Pan Noise Field=<0 0 1>)
   &SolveForcesAndVelocity(Clamp Velocity=false, Limit Acceleration=false)
   &DynamicMaterialParameters(Module.Param0WriteEnabled=true, Module.Param1WriteEnabled=true, Module.Param2WriteEnabled=true, Module.Param3WriteEnabled=true)
   &ScaleColor(
    Scale Alpha=&FloatFromCurve(
     FloatCurve=DataInterface() {propertyValues {Curve {keys [
         {arriveTangent=-1.4341, interpMode=RCIM_Cubic, time=0.1, value=0}
         {arriveTangent=0.0556, interpMode=RCIM_Cubic, leaveTangent=0.0556, time=0.3, value=1}
         {interpMode=RCIM_Cubic, tangentMode=RCTM_Break, time=1, value=0.05}
        ]}}}
     Scale Curve=&Multiply_Float(A=$User.Alpha, B=2)
    )
    Scale RGB=&MakeVectorFromLinearColorRGB(Color=&ColorFromCurve(ColorCurve=DataInterface() {propertyValues {
        AlphaCurve {keys []}
        BlueCurve {keys [{arriveTangent=-1.0847, interpMode=RCIM_Cubic, time=0, value=1}, <at 0.2431 0.3281 RCIM_Cubic>, <at 0.5613 1 RCIM_Cubic>]}
        GreenCurve {keys [{arriveTangent=-3.7763, interpMode=RCIM_Cubic, time=0, value=1}, <at 0.2431 0.3149 RCIM_Cubic>, <at 0.5613 1 RCIM_Cubic>]}
        RedCurve {keys [
          {arriveTangent=-8.1645, interpMode=RCIM_Cubic, time=0, value=0.9792}
          {arriveTangent=0.0371, interpMode=RCIM_Cubic, leaveTangent=0.0371, time=0.2431, value=0.2751}
          <at 0.5613 1 RCIM_Cubic>
         ]}
       }}))
    ScaleA=true
    ScaleRGB=true
   )
   &ScaleRibbonWidth(Ribbon Width Scale=&FloatFromCurve(FloatCurve=DataInterface() {propertyValues {Curve {keys [
         {arriveTangent=1.857, interpMode=RCIM_Cubic, leaveTangent=1.857, tangentMode=RCTM_User, time=0, value=0.25}
         {arriveTangent=0.1622, interpMode=RCIM_Cubic, leaveTangent=0.1622, tangentMode=RCTM_User, time=1, value=2}
        ]}}}))
   /SetVariables(
    Particles.RibbonID=/Niagara/DynamicInputs/NiagaraID/MakeNiagaraIDFromInt(Index:NiagaraInt32=$Particles.PartitionParticles.Partition, Acquire Tag:NiagaraInt32=0)
    Particles.RibbonLinkOrder:NiagaraFloat=$Particles.PartitionParticles.Normalized Index In Partition
   ) {enabled=false, isSetParameters=true}
  ]
  rendererClasses [&NiagaraRibbonRendererProperties]
  renderers [{rendererClass=&NiagaraRibbonRendererProperties}]
 }
}
summary {
 systemName=NS_Fountain_Base
 emitters [
  {bEnabled=false, emitterName=DropSplash, simTarget=GPUComputeSim, rendererClasses [&NiagaraSpriteRendererProperties]}
  {emitterName=TopSplash, simTarget=GPUComputeSim, rendererClasses [&NiagaraSpriteRendererProperties]}
  {bEnabled=false, emitterName=EdgeFlow, simTarget=GPUComputeSim, rendererClasses [&NiagaraSpriteRendererProperties]}
  {emitterName=Splash_SubUV, rendererClasses [&NiagaraSpriteRendererProperties]}
  {bEnabled=false, emitterName=DropSplash001, simTarget=GPUComputeSim, rendererClasses [&NiagaraSpriteRendererProperties]}
  {bEnabled=false, emitterName=WaterStream003, rendererClasses [&NiagaraRibbonRendererProperties]}
  {bEnabled=false, emitterName=SplashLinger001, rendererClasses [&NiagaraSpriteRendererProperties]}
 ]
 userVariables [
  User.Alpha {type=NiagaraFloat, defaultValue=0.2, description=""}
  User.BaseRadiusScale {type=NiagaraFloat, defaultValue=1, description=""}
  User.NumberofBeams {type=NiagaraInt32, defaultValue=16}
  User.NumberofRings {type=NiagaraInt32, defaultValue=3}
  User.RingOffset {
   type=NiagaraDataInterfaceArrayFloat3
   description=""
   defaultValue {
    struct=NiagaraExt_VariableValue_DataInterface
    value {
     dataInterface=&self:NiagaraDataInterfaceArrayFloat3_0
     dataInterfaceClass=&NiagaraDataInterfaceArrayFloat3
    }
   }
  }
  User.RingSize {
   type=NiagaraDataInterfaceArrayFloat
   defaultValue {
    struct=NiagaraExt_VariableValue_DataInterface
    value {
     dataInterface=&self:NiagaraDataInterfaceArrayFloat_0
     dataInterfaceClass=&NiagaraDataInterfaceArrayFloat
    }
   }
  }
  User.SplashOffset {
   type=NiagaraDataInterfaceArrayFloat3
   description=""
   defaultValue {
    struct=NiagaraExt_VariableValue_DataInterface
    value {
     dataInterface=&self:NiagaraDataInterfaceArrayFloat3_1
     dataInterfaceClass=&NiagaraDataInterfaceArrayFloat3
    }
   }
  }
  User.StreamBentPosition {type=NiagaraFloat, defaultValue=115, description=""}
  User.StreamBigPosition {type=NiagaraFloat, defaultValue=108, description=""}
  User.StreamSmallPosition {type=NiagaraFloat, defaultValue=35, description=""}
  User.TopSprayPosition {type=NiagaraFloat, defaultValue=150, description=""}
  User.TopSpraySpawnRate {type=NiagaraFloat, defaultValue=1, description=""}
 ]
}
