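# Niagara system for a one-spout fountain water effect: an arcing ribbon stream plus two sprite splashes, placed through user position parameters.
# The stream launches from StreamBentPos, shaped by gravity, point attraction and curl noise, widening over life.
# SubUV splashes spray in cones at SplashStartPos and, after an initial delay, SplashPos, growing under downward wind.
# User.Alpha, multiplied by five, scales the opacity fade curves of all three emitters.
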NiagaraSystem /Game/Effects/Fort_Effects/Effects/Environment/Fountain/NS_Fountain_OneSpout_Base
dependencies {
 usedDataInterfaces [&NiagaraDataInterfaceCurve, &NiagaraDataInterfaceVector2DCurve, &NiagaraDataInterfaceColorCurve]
 usedDynamicInputs []
 usedModules [&SystemState]
}
emitters {
 Splash_SubUV {
  emitterUpdateScript [
   &EmitterState(Delay First Loop Only=true, Life Cycle Mode=::NewEnumerator1, Loop Delay=1, Loop Duration=0, UseLoopDelay=true)
   &SpawnRate(SpawnRate=40, Use Spawn Probability=false)
  ]
  particleSpawnScript [
   &InitializeParticle(
    Lifetime Max=0.7
    Lifetime Min=0.3
    Lifetime Mode=::NewEnumerator1
    Sprite Size Max=<100 80>
    Sprite Size Min=<50 25>
    Sprite Size Mode=::NewEnumerator2
    UsePositionOffset=true
    Write Material Random=true
   ) {module=InitializeParticle001}
   &AddVelocityInCone(
    Cone Angle=50
    Cone Axis=<0 0 1>
    Module.Use Velocity Falloff On Cone Axis=false
    Velocity Distribution Along Cone Axis=0.25
    Velocity Strength=&UniformRangedFloat(Minimum=35, Maximum=100)
   )
   &SystemLocation(MaskSpawnGroup=false, Offset=$User.SplashPos)
   &SubUVAnimation(End Frame=3, SubUV Animation Mode:ENiagara_SubUVLookupMode=::NewEnumerator2)
  ]
  particleUpdateScript [
   &ParticleState()
   &ScaleSpriteSize(Scale Factor=&Vector2DFromCurve(
     Vector2Curve=DataInterface() {propertyValues {
       XCurve {keys [
         {arriveTangent=0.4604, interpMode=RCIM_Cubic, leaveTangent=0.4604, tangentMode=RCTM_User, time=0, value=0.05}
         {arriveTangent=1.2987, interpMode=RCIM_Cubic, leaveTangent=1.2987, tangentMode=RCTM_User, time=1, value=1}
        ]}
       YCurve {keys [
         {arriveTangent=2.0407, interpMode=RCIM_Cubic, leaveTangent=2.0407, tangentMode=RCTM_User, time=0, value=0.05}
         {arriveTangentWeight=0.3333, interpMode=RCIM_Cubic, leaveTangentWeight=0.0833, tangentMode=RCTM_User, time=0.75, value=1}
         {arriveTangent=-2.358, interpMode=RCIM_Cubic, leaveTangent=-2.358, tangentMode=RCTM_User, time=1, value=0.7}
        ]}
      }}
     ScaleCurve=<1 1>
    ))
   &DynamicMaterialParameters(
    Index 0 Param 1=&FloatFromCurve(
     FloatCurve=DataInterface() {propertyValues {Curve {keys [
         {arriveTangent=-0.5095, interpMode=RCIM_Cubic, leaveTangent=-0.5095, tangentMode=RCTM_User, time=0, value=1}
         {arriveTangent=-1.3585, interpMode=RCIM_Cubic, leaveTangent=-1.3585, tangentMode=RCTM_User, time=1, value=0}
        ]}}}
     Scale Curve=1.5
    )
    Module.Param0WriteEnabled=true
    Module.Param1WriteEnabled=false
    Module.Param2WriteEnabled=false
    Module.Param3WriteEnabled=false
   )
   &Color(Color=&MakeLinearColorFromVectorAndFloat(
     "Vector (RGB)"=<1 1 1>
     "Float (Alpha)"=&FloatFromCurve(
      FloatCurve=DataInterface() {propertyValues {Curve {keys [<at 0.4 1.0113 RCIM_Cubic>,<at 1.0011 0.002>]}}}
      Scale Curve=&Multiply_Float(A=$User.Alpha, B=5)
     )
    ))
   &Drag(Drag=0, Use Linear Drag=true, Use Rotational Drag=false)
   &WindForce(Air Resistance=0.1, Wind Speed=<0 0 -1>, Wind Speed Scale=100)
   &SpriteRotationRate()
   &SolveForcesAndVelocity(Clamp Velocity=false, Limit Acceleration=false)
  ]
  rendererClasses [&NiagaraSpriteRendererProperties]
  renderers [{rendererClass=&NiagaraSpriteRendererProperties}]
 }
 Splash_SubUV001 {
  emitterUpdateScript [
   &EmitterState()
   &SpawnRate(SpawnRate=30, Use Spawn Probability=false)
  ]
  particleSpawnScript [
   &InitializeParticle(
    Lifetime Max=0.7
    Lifetime Min=0.3
    Lifetime Mode=::NewEnumerator1
    Sprite Size Max=<100 80>
    Sprite Size Min=<50 25>
    Sprite Size Mode=::NewEnumerator2
    UsePositionOffset=true
    Write Material Random=true
   ) {module=InitializeParticle001}
   &AddVelocityInCone(
    Cone Angle=50
    Cone Axis=<1 0 1>
    Module.Use Velocity Falloff On Cone Axis=false
    Velocity Distribution Along Cone Axis=0.25
    Velocity Strength=&UniformRangedFloat(Minimum=35, Maximum=100)
   )
   &SystemLocation(MaskSpawnGroup=false, Offset=$User.SplashStartPos)
   &SubUVAnimation(End Frame=3, SubUV Animation Mode:ENiagara_SubUVLookupMode=::NewEnumerator2)
  ]
  particleUpdateScript [
   &ParticleState()
   &ScaleSpriteSize(Scale Factor=&Vector2DFromCurve(
     Vector2Curve=DataInterface() {propertyValues {
       XCurve {keys [
         {arriveTangent=0.4604, interpMode=RCIM_Cubic, leaveTangent=0.4604, tangentMode=RCTM_User, time=0, value=0.05}
         {arriveTangent=1.2987, interpMode=RCIM_Cubic, leaveTangent=1.2987, tangentMode=RCTM_User, time=1, value=1}
        ]}
       YCurve {keys [
         {arriveTangent=2.0407, interpMode=RCIM_Cubic, leaveTangent=2.0407, tangentMode=RCTM_User, time=0, value=0.05}
         {arriveTangentWeight=0.3333, interpMode=RCIM_Cubic, leaveTangentWeight=0.0833, tangentMode=RCTM_User, time=0.75, value=1}
         {arriveTangent=-2.358, interpMode=RCIM_Cubic, leaveTangent=-2.358, tangentMode=RCTM_User, time=1, value=0.7}
        ]}
      }}
     ScaleCurve=<1 1>
    ))
   &DynamicMaterialParameters(
    Index 0 Param 1=&FloatFromCurve(
     FloatCurve=DataInterface() {propertyValues {Curve {keys [
         {arriveTangent=-0.5095, interpMode=RCIM_Cubic, leaveTangent=-0.5095, tangentMode=RCTM_User, time=0, value=1}
         {arriveTangent=-1.3585, interpMode=RCIM_Cubic, leaveTangent=-1.3585, tangentMode=RCTM_User, time=1, value=0}
        ]}}}
     Scale Curve=1.5
    )
    Module.Param0WriteEnabled=true
    Module.Param1WriteEnabled=false
    Module.Param2WriteEnabled=false
    Module.Param3WriteEnabled=false
   )
   &Color(Color=&MakeLinearColorFromVectorAndFloat(
     "Vector (RGB)"=<1 1 1>
     "Float (Alpha)"=&FloatFromCurve(
      FloatCurve=DataInterface() {propertyValues {Curve {keys [<at 0.4 1.0113 RCIM_Cubic>,<at 1.0011 0.002>]}}}
      Scale Curve=&Multiply_Float(A=$User.Alpha, B=5)
     )
    ))
   &Drag(Drag=0, Use Linear Drag=true, Use Rotational Drag=false)
   &WindForce(Air Resistance=0.1, Wind Speed=<0 0 -1>, Wind Speed Scale=150)
   &SpriteRotationRate()
   &SolveForcesAndVelocity(Clamp Velocity=false, Limit Acceleration=false)
  ]
  rendererClasses [&NiagaraSpriteRendererProperties]
  renderers [{rendererClass=&NiagaraSpriteRendererProperties}]
 }
 WaterStream002 {
  emitterUpdateScript [
   &EmitterState(
    Life Cycle Mode=::NewEnumerator1
    Loop Duration=&UniformRangedFloat(Minimum=1, Maximum=1)
    Recalculate Duration Each Loop=true
    UseLoopDelay=false
   )
   &SpawnRate(SpawnRate=15, Use Spawn Probability=false)
  ]
  particleSpawnScript [
   /Niagara/Modules/Spawn/Initialization/InitializeParticle(
    Lifetime=0.75
    Module.Write Color=true
    Module.Write Lifetime=true
    Module.Write Mass=true
    Module.Write Material Random=true
    Module.Write Position=false
    Module.Write Scale=false
    Module.Write SpriteRotation=false
    Module.Write SpriteSize=true
    Position=&SimulationPosition()
   ) {enabled=false}
   &InitializeRibbon(Color=<rgba 1 1 1 0.2>, Ribbon Width=25, UsePositionOffset=false, Write Material Random=true)
   &AddVelocity(
    Module.Use Falloff Distance=false
    Origin Offset=<0.25 0 0.25>
    Override Seed=false
    Velocity Mode=::NewEnumerator1
    Velocity Speed=&RandomRangeFloat(Minimum=-260, Maximum=-280)
   )
   &SystemLocation(MaskSpawnGroup=false, Offset=$User.StreamBentPos)
  ]
  particleUpdateScript [
   &ParticleState()
   &GravityForce(Gravity=<0 0 -1800>)
   &PointAttractionForce(
    Attraction Radius=250
    Attraction Strength=2.5
    Falloff Exponent=1
    Kill Within Radius=false
    Module.Use Falloff=true
    UseAttractorOffset=true
   )
   &CurlNoiseForce(Noise Frequency=25, Noise Strength=5, Pan Noise=true, Pan Noise Field=<0 0 1>)
   &SolveForcesAndVelocity(Clamp Velocity=false, Limit Acceleration=false)
   &DynamicMaterialParameters(Module.Param0WriteEnabled=true, Module.Param1WriteEnabled=true, Module.Param2WriteEnabled=true, Module.Param3WriteEnabled=true)
   &ScaleColor(
    Scale Alpha=&FloatFromCurve(
     FloatCurve=DataInterface() {propertyValues {Curve {keys [
         {arriveTangent=-1.4341, interpMode=RCIM_Cubic, time=0.1, value=0}
         {arriveTangent=1.4286, leaveTangent=1.4286, time=0.3, value=1}
         {interpMode=RCIM_Cubic, tangentMode=RCTM_User, time=0.9, value=1}
         <at 0.95 0>
         {interpMode=RCIM_Cubic, tangentMode=RCTM_Break, time=1, value=0}
        ]}}}
     Scale Curve=&Multiply_Float(A=$User.Alpha, B=5)
    )
    Scale RGB=&MakeVectorFromLinearColorRGB(Color=&ColorFromCurve(ColorCurve=DataInterface() {propertyValues {
        AlphaCurve {keys []}
        BlueCurve {keys [{arriveTangent=-1.0847, interpMode=RCIM_Cubic, time=0, value=1}, <at 0.2431 0.3281 RCIM_Cubic>, <at 0.5613 1 RCIM_Cubic>]}
        GreenCurve {keys [{arriveTangent=-3.7763, interpMode=RCIM_Cubic, time=0, value=1}, <at 0.2431 0.3149 RCIM_Cubic>, <at 0.5613 1 RCIM_Cubic>]}
        RedCurve {keys [
          {arriveTangent=-8.1645, interpMode=RCIM_Cubic, time=0, value=0.9792}
          {arriveTangent=0.0371, interpMode=RCIM_Cubic, leaveTangent=0.0371, time=0.2431, value=0.2751}
          <at 0.5613 1 RCIM_Cubic>
         ]}
       }}))
    ScaleA=true
    ScaleRGB=true
   )
   &ScaleRibbonWidth(Ribbon Width Scale=&FloatFromCurve(FloatCurve=DataInterface() {propertyValues {Curve {keys [
         {arriveTangent=1.857, interpMode=RCIM_Cubic, leaveTangent=1.857, tangentMode=RCTM_User, time=0, value=0.25}
         {arriveTangent=0.1622, interpMode=RCIM_Cubic, leaveTangent=0.1622, tangentMode=RCTM_User, time=1, value=1}
        ]}}}))
   /SetVariables(
    Particles.RibbonID=/Niagara/DynamicInputs/NiagaraID/MakeNiagaraIDFromInt(Index:NiagaraInt32=$Particles.PartitionParticles.Partition, Acquire Tag:NiagaraInt32=0)
    Particles.RibbonLinkOrder:NiagaraFloat=$Particles.PartitionParticles.Normalized Index In Partition
   ) {enabled=false, isSetParameters=true}
  ]
  rendererClasses [&NiagaraRibbonRendererProperties]
  renderers [{rendererClass=&NiagaraRibbonRendererProperties}]
 }
}
summary {
 systemName=NS_Fountain_OneSpout_Base
 emitters [
  {emitterName=Splash_SubUV, rendererClasses [&NiagaraSpriteRendererProperties]}
  {emitterName=WaterStream002, rendererClasses [&NiagaraRibbonRendererProperties]}
  {emitterName=Splash_SubUV001, rendererClasses [&NiagaraSpriteRendererProperties]}
 ]
 userVariables [
  User.Alpha {type=NiagaraFloat, defaultValue=0.2, description=""}
  User.SplashPos {type=Vector3f, description="", defaultValue {struct=Vector3f, value=<0 0 0>}}
  User.SplashStartPos {type=Vector3f, description="", defaultValue {struct=Vector3f, value=<0 0 0>}}
  User.StreamBentPos {type=Vector3f, description="", defaultValue {struct=Vector3f, value=<0 0 0>}}
  User.Vector {type=Vector3f, description="", defaultValue {struct=Vector3f, value=<0 0 0>}}
 ]
}
