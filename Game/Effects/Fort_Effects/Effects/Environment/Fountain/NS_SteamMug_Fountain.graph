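# Niagara system for a steam mug fountain effect, with one GPU emitter drawing brown-tinted splash particles as sub-UV sprites and meshes.
# Particles are partitioned onto concentric rings driven by System ring, beam and radius variables, then launched upward in a narrow cone under gravity.
# Over life they grow, fade and animate a material parameter, with opacity scaled by time of day.
# A second position and velocity pair is solved separately each frame.
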NiagaraSystem /Game/Effects/Fort_Effects/Effects/Environment/Fountain/NS_SteamMug_Fountain
dependencies {
 usedDataInterfaces [&NiagaraDataInterfaceCurve, &NiagaraDataInterfaceArrayFloat3, &NiagaraDataInterfaceArrayFloat]
 usedDynamicInputs [&SelectVectorFromArray]
 usedModules [
  &self:SystemSpawnScript.SystemScriptSource.SystemScriptGraph.NiagaraNodeAssignment_0.SetVariables_#1
  &self:SystemSpawnScript.SystemScriptSource.SystemScriptGraph.NiagaraNodeAssignment_2.SetVariables_#2
  &self:SystemSpawnScript.SystemScriptSource.SystemScriptGraph.NiagaraNodeAssignment_1.SetVariables_#3
  &self:SystemSpawnScript.SystemScriptSource.SystemScriptGraph.NiagaraNodeAssignment_4.SetVariables_#4
  &self:SystemSpawnScript.SystemScriptSource.SystemScriptGraph.NiagaraNodeAssignment_5.SetVariables_#5
  &self:SystemSpawnScript.SystemScriptSource.SystemScriptGraph.NiagaraNodeAssignment_6.SetVariables_#6
  &SystemState
 ]
}
emitters {Splash_SubUV {
  simTarget=GPUComputeSim
  emitterSpawnScript [/SetVariables() {isSetParameters=true}]
  emitterUpdateScript [
   &EmitterState()
   &SpawnRate(SpawnRate=64, Use Spawn Probability=false)
  ]
  particleSpawnScript [
   &InitializeParticle(
    Color=<rgba 0.3217 0.1942 0.1156 1>
    Lifetime Max=1
    Lifetime Min=0.6
    Lifetime Mode=::NewEnumerator1
    Mesh Scale Mode=::NewEnumerator1
    Mesh Uniform Scale Max=1
    Mesh Uniform Scale Min=0.5
    Sprite Size Max=<50 50>
    Sprite Size Min=<25 25>
    Sprite Size Mode=::NewEnumerator2
    UsePositionOffset=true
    Write Material Random=true
   ) {module=InitializeParticle001}
   &AddVelocityInCone(
    Cone Angle=5
    Cone Axis=<0 0 1>
    Module.Use Velocity Falloff On Cone Axis=false
    Velocity Distribution Along Cone Axis=0.25
    Velocity Strength=&UniformRangedFloat(Minimum=500, Maximum=50)
   )
   &SystemLocation(MaskSpawnGroup=false, Offset=<0 0 -80>)
   &PartitionParticles(
    Number Of Partitions=$System.NumberOfRings
    Particles Per Partition:NiagaraInt32=$System.NumberOfBeams
    Partition Mode:ENiagara_PartitionMode=::NewEnumerator1
   )
   &ShapeLocation(
    MaskSpawnGroup=false
    Override Seed=false
    Ring / Disc Distribution Mode:ENiagaraSphereDistributionMode=::NewEnumerator1
    Ring Radius=&Multiply_Float(
     A=&SelectFloatFromArray(Direct Array Index=$Particles.PartitionParticles.Partition, Float Selection Array=$System.RingSize)
     B=$System.BaseRadiusScale
    )
    Shape Primitive=::NewEnumerator4
    U Position=$Particles.PartitionParticles.Normalized Index In Partition
    UseOwnerScale=true
   )
   &SubUVAnimation(End Frame=3, SubUV Animation Mode:ENiagara_SubUVLookupMode=::NewEnumerator2)
   &InitialMeshOrientation(
    Mesh Orientation Mode=::NewEnumerator1
    Rotation=&RandomRangeVector(Minimum=<0 0 0>, Maximum=<0 0 1>)
    Use Rotation Vector=true
   )
   &ShapeLocation(MaskSpawnGroup=false, Non Uniform Scale=<1 1 0>, Override Seed=false, Sphere Radius=2, UseOwnerScale=true) {module=ShapeLocation001}
  ]
  particleUpdateScript [
   &ParticleState()
   &ScaleSpriteSize(Scale Factor=&Multiply_Vector2D_ByFloat(
     Float=&FloatFromCurve(
      FloatCurve=DataInterface() {propertyValues {Curve {keys [
          {interpMode=RCIM_Cubic, tangentMode=RCTM_User, time=0, value=0}
          {arriveTangent=1.4313, interpMode=RCIM_Cubic, leaveTangent=1.4313, tangentMode=RCTM_User, time=0.6, value=2}
          {tangentMode=RCTM_User, time=1, value=1.5}
         ]}}}
      Scale Curve=4
     )
     Vector2D=&Multiply_Vector2D_ByFloat(Float=&Multiply_Float(A=$System.MaxSizeMult, B=0.035), Vector2D=<1 1>)
    ))
   &DynamicMaterialParameters(
    Index 0 Param 1=&FloatFromCurve(
     FloatCurve=DataInterface() {propertyValues {Curve {keys [
         {arriveTangent=-0.5095, interpMode=RCIM_Cubic, leaveTangent=-0.5095, tangentMode=RCTM_User, time=0, value=1}
         {arriveTangent=-1.3585, interpMode=RCIM_Cubic, leaveTangent=-1.3585, tangentMode=RCTM_User, time=1, value=0}
        ]}}}
     Scale Curve=1.5
    )
    Module.Param0WriteEnabled=true
    Module.Param1WriteEnabled=false
    Module.Param2WriteEnabled=false
    Module.Param3WriteEnabled=false
   )
   &SpriteRotationRate()
   &GravityForce(Gravity=<0 0 -100>)
   &SolveForcesAndVelocity(Clamp Velocity=true, Limit Acceleration=false, Speed Limit=500)
   /SetVariables(
    Particles.Position2:NiagaraPosition=&AddVectorToPosition(Position=$Particles.Position, Vector=<0 0 0>)
    Particles.Velocity2:Vector3f=<0 0 1>
   ) {isSetParameters=true}
   &CalculateAccurateVelocity(Calculate Accurate Velocity On First Frame=true) {enabled=false}
   &SolveForcesAndVelocity(
    Clamp Velocity=true
    Limit Acceleration=false
    Position:NiagaraPosition=$Particles.Position2
    Speed Limit=5000
    Velocity:Vector3f=$Particles.Velocity2
   ) {module=SolveForcesAndVelocity001}
   &ScaleColor(
    Scale Alpha=&FloatFromCurve(
     FloatCurve=DataInterface() {propertyValues {Curve {keys [<at 0 0 RCIM_Cubic>,<at 0.2 1 RCIM_Cubic>,<at 1.0011 0>]}}}
     Scale Curve=&Multiply_Float(
      A=$System.Alpha
      B=&FloatFromCurve(
       FloatCurve=DataInterface() {propertyValues {Curve {keys [<at 0 0.25>,<at 0.2 0.5>,<at 0.75 0.5>,<at 1 0.25>]}}}
       CurveIndex=Linked() {linkedVariable {name=NPC.FortniteNPC.TimeOfDay_0-1, type=NiagaraFloat}}
       Scale Curve=15
      )
     )
    )
    ScaleA=true
    ScaleRGB=true
   )
  ]
  rendererClasses [&NiagaraSpriteRendererProperties, &NiagaraMeshRendererProperties]
  renderers [
   {rendererClass=&NiagaraSpriteRendererProperties}
   {rendererClass=&NiagaraMeshRendererProperties, rendererIndex=1}
  ]
 }}
summary {
 systemName=NS_SteamMug_Fountain
 emitters [{
   emitterName=Splash_SubUV
   simTarget=GPUComputeSim
   rendererClasses [&NiagaraSpriteRendererProperties, &NiagaraMeshRendererProperties]
  }]
 userVariables []
}
